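; Neon sign Blueprint for NeonCity whose emissive glow can switch off during the day.
; InitSign, run at construction and after a random BeginPlay delay, creates dynamic materials for M_SimpleNeonSign slots, applying color, brightness, flicker and scale-adjusted thickness.
; With Nanite enabled it uses the override material and disables World Position Offset.
; Day-phase changes toggle emissive after random staggering; a custom Update event refreshes all signs from the material-collection time of day.
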(event EventBeginPlay
  (Utilities|FlowControl|Delay (Math|Random|RandomFloatinRange 0.0 0.5))
  (CallFunction|InitSign))

(event EventOnDayPhaseChanged
  (params (CurrentDayPhase :type EFortDayPhase%20Enum :default "Morning") (PreviousDayPhase :type EFortDayPhase%20Enum :default "Morning") (bAtCreation :type Boolean))
  (if (Variables|Default|GetTurns%20off%20At%20Day)
    (Utilities|FlowControl|Delay (Math|Random|RandomFloatinRange 0.0 5.0))
    (if (Utilities|Enum|Equal(Enum) CurrentDayPhase "Day")
      (Rendering|Material|SetScalarParameterValueonMaterials (Variables|Debug|GetStaticMeshComponent) "EmissiveMult" 0.0)
      (else
        _)))
  (Rendering|Material|SetScalarParameterValueonMaterials (Variables|Debug|GetStaticMeshComponent) "EmissiveMult" 1.0))

(event Collision|EventActorBeginOverlap
  (params (OtherActor :type Actor%20Object%20Reference)))

(event EventTick
  (params (DeltaSeconds :type Float%20%28single-precision%29)))

(event Custom|Update
  (bind _outactors (Actor|GetAllActorsOfClass "/Game/Environments/Asteria/Sets/NeonCity/Techart/NeonSigns/BP_NeonSign.BP_NeonSign_C"))
  (for _array_element _outactors
    (if (Variables|Default|GetTurns%20off%20At%20Day _array_element)
      (bind _returnvalue (Rendering|Material|GetScalarParameterValue "/Game/Packages/Fortress_SharedMaterials/GlobalMaterialParameters/FortniteMaterialParameters.FortniteMaterialParameters" "TimeOfDay_0-1" :!class /Script/Engine.KismetMaterialLibrary))
      (if (and (> 2e+01 (* _returnvalue 24.0)) (> (* _returnvalue 24.0) 8.0))
        (Rendering|Material|SetScalarParameterValueonMaterials (Variables|/Script/FortniteGame.BuildingSMActor|GetStaticMeshComponent _array_element) "EmissiveMult" 0.0)
        (else
          _))
      (else
        _))
    (Rendering|Material|SetScalarParameterValueonMaterials (Variables|/Script/FortniteGame.BuildingSMActor|GetStaticMeshComponent _array_element) "EmissiveMult" 1.0)))

(fn InitSign
  (bind _staticmeshcomponent (Variables|Debug|GetStaticMeshComponent))
  (bind _returnvalue (Development|GetConsoleVariableBoolValue "r.nanite"))
  (Utilities|FlowControl|Sequence
    (:then_0
      (for (_array_element _array_index) (Variables|/Script/Engine.StaticMesh|GetStaticMaterials (Variables|/Script/Engine.StaticMeshComponent|GetStaticMesh (Variables|Debug|GetStaticMeshComponent)))
        (bind _returnvalue_2 (Rendering|Material|GetBaseMaterial (Utilities|Struct|BreakStaticMaterial _array_element)))
        (if (== _returnvalue_2 "/Game/Environments/Asteria/Sets/NeonCity/Techart/NeonSigns/M_SimpleNeonSign.M_SimpleNeonSign")
          (bind _returnvalue_3 (Rendering|Material|CreateDynamicMaterialInstance (Variables|Debug|GetStaticMeshComponent) _array_index (select _returnvalue (Rendering|Material|GetNaniteOverideMaterial _returnvalue_2) (Utilities|Struct|BreakStaticMaterial _array_element)) "None" :!class /Script/Engine.PrimitiveComponent))
          (Rendering|Material|SetVectorParameterValue _returnvalue_3 "Sign Color A" (Variables|Default|GetColor%20A) :!class /Script/Engine.MaterialInstanceDynamic)
          (Rendering|Material|SetVectorParameterValue _returnvalue_3 "Sign Color B" (Variables|Default|GetColor%20B) :!class /Script/Engine.MaterialInstanceDynamic)
          (Rendering|Material|SetScalarParameterValue _returnvalue_3 "Sign Brightness" (Variables|Default|GetBrightness) :!class /Script/Engine.MaterialInstanceDynamic)
          (Rendering|Material|SetScalarParameterValue _returnvalue_3 "Blink Color B" (Math|Conversions|ToFloat(Boolean) (Variables|Default|GetPulse%20Color%20B)) :!class /Script/Engine.MaterialInstanceDynamic)
          (Rendering|Material|SetScalarParameterValue _returnvalue_3 "LumenBoost" (Variables|Default|GetLumenBoost) :!class /Script/Engine.MaterialInstanceDynamic)
          (Rendering|Material|SetScalarParameterValue _returnvalue_3 "Flickering" (Math|Conversions|ToFloat(Boolean) (Variables|Default|GetFlickering)) :!class /Script/Engine.MaterialInstanceDynamic)
          (Rendering|Material|SetScalarParameterValue _returnvalue_3 "ST_WidthBase" (* (Variables|Default|GetSplineThickness) (.x (.scale (Transformation|GetActorTransform)))) :!class /Script/Engine.MaterialInstanceDynamic))))
    (:then_1
      (if _returnvalue
        (Rendering|LOD|SetEvaluateWorldPositionOffset _staticmeshcomponent false :!class /Script/Engine.StaticMeshComponent)
        (else
          (Rendering|LOD|SetEvaluateWorldPositionOffset _staticmeshcomponent true :!class /Script/Engine.StaticMeshComponent))))))

(fn ConstructionScript
  (CallFunction|InitSign))
