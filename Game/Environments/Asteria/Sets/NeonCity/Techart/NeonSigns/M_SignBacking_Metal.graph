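# M_SignBacking_Metal is a Nanite-enabled, Default Lit material for the metal backing of neon signs in the NeonCity set.
# A world-aligned grime texture, scaled by Grime Size, blends between minimum and maximum roughness values and is then saturated.
# It also holds a Metal Color parameter, a normal map parameter, a few constants and a destructible damage function.
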Material /Game/Environments/Asteria/Sets/NeonCity/Techart/NeonSigns/M_SignBacking_Metal
graph {
 properties {bUsedWithNanite=true, pixelDepthOffsetMode=PDOM_Legacy, usedShadingModels=Default Lit}
 nodes [
  Constant3Vector() {constant=<rgba 0 1 0 0>}
  Constant() {r=1}
  Constant() {r=1}
  MaterialFunctionCall() {materialFunction=&MF_FN_DestructibleDamage}
  n1 = TextureObject() {samplerType=SAMPLERTYPE_LinearColor, texture=/Game/Environments/Apollo/Sets/Tinfoil/Textures/T_Tinfoil_GlassGrime_M}
  n2 = ScalarParameter() {defaultValue=750, parameterName=Grime Size}
  Saturate(None=LinearInterpolate(
    A=ScalarParameter() {defaultValue=0.316, parameterName=Roughness Min}
    B=ScalarParameter() {defaultValue=0.5425, parameterName=Roughness Max}
    Alpha=MaterialFunctionCall(TextureObject=@n1, TextureSize=@n2) {materialFunction=/Game/Weapons/WeaponSkins/Wraps/MaterialFunctions/WorldAlignedTexture}.XYZ Texture
   ))
  TextureSampleParameter2D() {parameterName=N, samplerType=SAMPLERTYPE_Normal, texture=&/NSign_N}
  VectorParameter() {parameterName=Metal Color, defaultValue=<rgba 0.2115 0.2424 0.2448 1>}
 ]
}
