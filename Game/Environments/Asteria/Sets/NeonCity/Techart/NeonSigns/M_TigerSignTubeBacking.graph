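# Opaque Default Lit, Nanite-enabled material for the tube backing of the tiger neon sign in the NeonCity set.
# A tiled TigerBG mask drives scalar blends with its red channel and gates the animated glow, with a matching normal map sampled alongside.
# A panning vertical gradient indexes a curve atlas row for glow color, which the mask's green channel modulates before NeonCity's emissive balance function.
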Material /Game/Environments/Asteria/Sets/NeonCity/Techart/NeonSigns/M_TigerSignTubeBacking
graph {
 properties {bUsedWithNanite=true, pixelDepthOffsetMode=PDOM_Legacy, usedShadingModels=Default Lit}
 nodes [
  Constant3Vector() {constant=<rgba 0.1719 0.1719 0.1719 0>}
  n1 = Multiply(A=TextureCoordinate(), B=ScalarParameter() {defaultValue=3, parameterName=TubeTiling})
  n2 = TextureSample(UVs=@n1) {samplerType=SAMPLERTYPE_LinearColor, texture=&/TigerBG_M}
  LinearInterpolate(Alpha=@n2.R) {constA=1, constB=0}
  LinearInterpolate(Alpha=@n2.R) {constA=0.25, constB=0}
  n3 = ScalarParameter() {parameterName=PanX}
  n4 = ScalarParameter() {defaultValue=0.1, parameterName=PanY}
  n5 = TextureSample(UVs=Panner(Coordinate=TextureCoordinate() {uTiling=0.5, vTiling=0.5}, Speed=AppendVector(A=@n3, B=@n4)) {speedX=1, speedY=1}) {
   samplerSource=SSM_Wrap_WorldGroupSettings
   samplerType=SAMPLERTYPE_LinearColor
   texture=/Game/Global/Textures/Default/Masks_and_Gradients/T_VertGradient
  }
  n6 = CurveAtlasRowParameter(CurveTime=@n5.R) {
   atlas=/Game/Environments/Asteria/Sets/NeonCity/Techart/Holograms/NewHologramMaterial/HologramCurves
   curve=/Game/Athena/LinearColorCurves/Curve_WallArtTiger
   defaultValue=3
   parameterName=Curves
  }
  n7 = Multiply(A=@n2.G, B=@n6)
  LinearInterpolate(
   B=MaterialFunctionCall!(Emissive=Multiply(B=@n7) {constA=0.5}) {materialFunction=/Game/Environments/Asteria/Sets/NeonCity/Techart/Screen/MaterialFunctions/MF_NeonCityEmissiveBalance}.Result
   Alpha=@n2.R
  )
  TextureSample(UVs=@n1) {samplerType=SAMPLERTYPE_Normal, texture=&/TigerBG_N}
 ]
}
