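# Nanite-enabled Default Lit material for NeonCity neon sign tubes, producing a scrolling color gradient glow.
# A half-tiled vertical gradient texture pans at PanX and PanY parameter speeds, vertical only by default.
# Its red channel indexes a color row in the shared HologramCurves atlas, selected by the Curves parameter.
# The color is scaled down and fed as emissive into MF_NeonCityEmissiveBalance.
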Material /Game/Environments/Asteria/Sets/NeonCity/Techart/NeonSigns/M_Tube_Solid_VerticalGradientPanning
graph {
 properties {bUsedWithNanite=true, pixelDepthOffsetMode=PDOM_Legacy, usedShadingModels=Default Lit}
 nodes [
  Constant()
  n1 = ScalarParameter() {parameterName=PanX}
  n2 = ScalarParameter() {defaultValue=0.1, parameterName=PanY}
  n3 = TextureSample(UVs=Panner(Coordinate=TextureCoordinate() {uTiling=0.5, vTiling=0.5}, Speed=AppendVector(A=@n1, B=@n2)) {speedX=1, speedY=1}) {
   samplerSource=SSM_Wrap_WorldGroupSettings
   samplerType=SAMPLERTYPE_LinearColor
   texture=/Game/Global/Textures/Default/Masks_and_Gradients/T_VertGradient
  }
  n4 = CurveAtlasRowParameter(CurveTime=@n3.R) {
   atlas=/Game/Environments/Asteria/Sets/NeonCity/Techart/Holograms/NewHologramMaterial/HologramCurves
   curve=/Game/Athena/LinearColorCurves/Curve_WallArtTiger
   defaultValue=3
   parameterName=Curves
  }
  MaterialFunctionCall!(Emissive=Multiply(A=@n4) {constB=0.1}) {materialFunction=/Game/Environments/Asteria/Sets/NeonCity/Techart/Screen/MaterialFunctions/MF_NeonCityEmissiveBalance}
 ]
}
